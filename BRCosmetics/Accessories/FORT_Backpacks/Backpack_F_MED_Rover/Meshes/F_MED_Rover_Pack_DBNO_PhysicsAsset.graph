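# Physics asset for the DBNO variant of the F_MED_Rover backpack cosmetic, defining jiggle physics for its loops, carabiner, keychain and straps.
# A large capsule on the pack bone is the hub, with a small capsule MSTR anchor per accessory constrained to it with free swing and twist.
# Loops, carabiner and keychain hang as single angle-limited bodies from their anchors.
# Top, bottom and key straps form two-segment chains with tight angular limits.
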PhysicsAsset /BRCosmetics/Accessories/FORT_Backpacks/Backpack_F_MED_Rover/Meshes/F_MED_Rover_Pack_DBNO_PhysicsAsset
bodies {
 dyn_botLoop [{
   shapeType=Capsule
   length=5.2269
   radius=1.6908
   shapeName=dyn_botLoop_capsule
   center=<4.0917 -0.22 -0.1712>
   rotation=<rot 31.839 -90 10.605>
  }]
 dyn_botLoop_MSTR [{shapeType=Capsule, length=0.505, radius=0.505, shapeName=dyn_botLoop_MSTR_capsule, rotation=<rot 0 -90 0>}]
 dyn_botStrap_00_MSTR [{shapeType=Capsule, length=0.505, radius=0.505, shapeName=dyn_botStrap_00_MSTR_capsule, rotation=<rot 0 -90 0>}]
 dyn_botStrap_01 [{
   shapeType=Capsule
   length=1.1668
   radius=0.7541
   shapeName=dyn_botStrap_01_capsule
   center=<0.7111 0.0381 -0.2451>
   rotation=<rot 0 163.9831 -85.9874>
  }]
 dyn_botStrap_02 [{
   shapeType=Capsule
   length=1.9561
   radius=1.66
   shapeName=dyn_botStrap_02_capsule
   center=<1.408 -0.0192 -0.0093>
   rotation=<rot 0 -52.4741 -88.1692>
  }]
 dyn_carabiner [{
   shapeType=Capsule
   length=4.2794
   radius=2.3455
   shapeName=dyn_carabiner_capsule
   center=<2.2482 -1.0151 -0.4351>
   rotation=<rot 0 -142.1599 -76.0175>
  }]
 dyn_carabiner_MSTR [{shapeType=Capsule, length=0.505, radius=0.505, shapeName=dyn_carabiner_MSTR_capsule, rotation=<rot 0 -90 0>}]
 dyn_keyStrap_00_MSTR [{shapeType=Capsule, length=0.505, radius=0.505, shapeName=dyn_keyStrap_00_MSTR_capsule, rotation=<rot 0 -90 0>}]
 dyn_keyStrap_01 [{
   shapeType=Capsule
   length=0.505
   radius=0.505
   shapeName=dyn_keyStrap_01_capsule
   center=<1.4324 0.0548 0.0164>
   rotation=<rot 0 -91.2207 -89.602>
  }]
 dyn_keyStrap_02 [{
   shapeType=Capsule
   length=0.959
   radius=0.3617
   shapeName=dyn_keyStrap_02_capsule
   center=<1.5068 0.05 0.0368>
   rotation=<rot 0 -92.606 -88.2939>
  }]
 dyn_keychain [{
   shapeType=Capsule
   length=3.3992
   radius=1.1322
   shapeName=dyn_keychain_capsule
   center=<3.1578 -0.0687 -0.0238>
   rotation=<rot 0 -98.8376 -88.7334>
  }]
 dyn_keychain_MSTR [{shapeType=Capsule, length=0.505, radius=0.505, shapeName=dyn_keychain_MSTR_capsule, rotation=<rot 0 -90 0>}]
 dyn_topLoop [{
   shapeType=Capsule
   length=3.9001
   radius=1.1717
   shapeName=dyn_topLoop_capsule
   center=<3.8261 -0.3967 -0.1857>
   rotation=<rot 33.9624 -90 12.9257>
  }]
 dyn_topLoop_MSTR [{shapeType=Capsule, length=0.505, radius=0.505, shapeName=dyn_topLoop_MSTR_capsule, rotation=<rot 0 -90 0>}]
 dyn_topStrap_00_MSTR [{shapeType=Capsule, length=0.505, radius=0.505, shapeName=dyn_topStrap_00_MSTR_capsule, rotation=<rot 0 -90 0>}]
 dyn_topStrap_01 [{
   shapeType=Capsule
   length=1.2011
   radius=0.7101
   shapeName=dyn_topStrap_01_capsule
   center=<1.124 0.0164 -0.3582>
   rotation=<rot 0 172.3456 -89.9928>
  }]
 dyn_topStrap_02 [{
   shapeType=Capsule
   length=2.1186
   radius=1.4828
   shapeName=dyn_topStrap_02_capsule
   center=<1.7519 -0.0318 -0.14>
   rotation=<rot 0 101.0717 -89.7132>
  }]
 pack [{
   shapeType=Capsule
   length=25.3857
   radius=9.1036
   shapeName=pack_capsule
   center=<1.9307 -5.8376 -3.5157>
   rotation=<rot -4.2615 -90 -23.4599>
  }]
}
constraints [
 {
  bone1Name=dyn_keychain_MSTR
  bone2Name=pack
  swing1LimitDegrees=45
  swing1Motion=Free
  swing2LimitDegrees=45
  swing2Motion=Free
  twistLimitDegrees=45
  twistMotion=Free
 }
 {bone1Name=dyn_keychain, bone2Name=dyn_keychain_MSTR, swing1LimitDegrees=60, swing2LimitDegrees=90, twistLimitDegrees=2}
 {
  bone1Name=dyn_topLoop_MSTR
  bone2Name=pack
  swing1LimitDegrees=45
  swing1Motion=Free
  swing2LimitDegrees=45
  swing2Motion=Free
  twistLimitDegrees=45
  twistMotion=Free
 }
 {bone1Name=dyn_topLoop, bone2Name=dyn_topLoop_MSTR, swing1LimitDegrees=10, swing2LimitDegrees=10, twistLimitDegrees=2.5}
 {
  bone1Name=dyn_botLoop_MSTR
  bone2Name=pack
  swing1LimitDegrees=45
  swing1Motion=Free
  swing2LimitDegrees=45
  swing2Motion=Free
  twistLimitDegrees=45
  twistMotion=Free
 }
 {bone1Name=dyn_botLoop, bone2Name=dyn_botLoop_MSTR, swing1LimitDegrees=10, swing2LimitDegrees=10, twistLimitDegrees=2.5}
 {
  bone1Name=dyn_carabiner_MSTR
  bone2Name=pack
  swing1LimitDegrees=45
  swing1Motion=Free
  swing2LimitDegrees=45
  swing2Motion=Free
  twistLimitDegrees=45
  twistMotion=Free
 }
 {bone1Name=dyn_carabiner, bone2Name=dyn_carabiner_MSTR, swing1LimitDegrees=15, swing2LimitDegrees=45, twistLimitDegrees=5}
 {
  bone1Name=dyn_botStrap_00_MSTR
  bone2Name=pack
  swing1LimitDegrees=45
  swing1Motion=Free
  swing2LimitDegrees=45
  swing2Motion=Free
  twistLimitDegrees=45
  twistMotion=Free
 }
 {bone1Name=dyn_botStrap_01, bone2Name=dyn_botStrap_00_MSTR, swing1LimitDegrees=5, swing2LimitDegrees=20, twistLimitDegrees=5}
 {bone1Name=dyn_botStrap_02, bone2Name=dyn_botStrap_01, swing1LimitDegrees=5, swing2LimitDegrees=20, twistLimitDegrees=5}
 {
  bone1Name=dyn_topStrap_00_MSTR
  bone2Name=pack
  swing1LimitDegrees=45
  swing1Motion=Free
  swing2LimitDegrees=45
  swing2Motion=Free
  twistLimitDegrees=45
  twistMotion=Free
 }
 {bone1Name=dyn_topStrap_01, bone2Name=dyn_topStrap_00_MSTR, swing1LimitDegrees=5, swing2LimitDegrees=20, twistLimitDegrees=5}
 {bone1Name=dyn_topStrap_02, bone2Name=dyn_topStrap_01, swing1LimitDegrees=5, swing2LimitDegrees=20, twistLimitDegrees=5}
 {
  bone1Name=dyn_keyStrap_00_MSTR
  bone2Name=pack
  swing1LimitDegrees=45
  swing1Motion=Free
  swing2LimitDegrees=45
  swing2Motion=Free
  twistLimitDegrees=45
  twistMotion=Free
 }
 {bone1Name=dyn_keyStrap_01, bone2Name=dyn_keyStrap_00_MSTR, swing1LimitDegrees=10, swing2LimitDegrees=10, twistLimitDegrees=10}
 {bone1Name=dyn_keyStrap_02, bone2Name=dyn_keyStrap_01, swing1LimitDegrees=10, swing2LimitDegrees=10, twistLimitDegrees=10}
]
